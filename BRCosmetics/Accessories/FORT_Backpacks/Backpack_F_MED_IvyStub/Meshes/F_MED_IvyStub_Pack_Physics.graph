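# Physics asset for the IvyStub backpack cosmetic, giving its small limbed figure jointed physics bodies for its arms and legs.
# Each bone gets one small capsule, with four three-link chains (shoulder-elbow-wrist and leg-knee-ankle on each side) hanging off base_COG.
# Constraints lock twist and allow only narrow swing, so the limbs have limited wobble rather than full ragdoll freedom.
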PhysicsAsset /BRCosmetics/Accessories/FORT_Backpacks/Backpack_F_MED_IvyStub/Meshes/F_MED_IvyStub_Pack_Physics
bodies {
 base_COG [{shapeType=Capsule, length=3, radius=2, center=<0.4893 0 0>, rotation=<rot 90 0 0>}]
 dyn_Knee_L1 [{shapeType=Capsule, length=1.5, radius=1.2, center=<2.7033 0 0>, rotation=<rot 90 0 0>}]
 dyn_Knee_R [{shapeType=Capsule, length=1.7, radius=1, center=<1.3138 0 0>, rotation=<rot 90 0.0001 0>}]
 dyn_Leg_L [{shapeType=Capsule, length=2, radius=1.2, center=<1.9004 0 0>, rotation=<rot 90 0 0>}]
 dyn_Reg_R [{shapeType=Capsule, length=2.3, radius=1.3, center=<0.6909 0 0>, rotation=<rot 90 0 0>}]
 dyn_ankle_L [{shapeType=Capsule, length=1.2, radius=1, center=<0.9431 0 0>, rotation=<rot 90 0 0>}]
 dyn_ankle_R [{shapeType=Capsule, length=1.5, radius=1, center=<0.4851 0 0>, rotation=<rot 90 0 0>}]
 dyn_elbow_L [{shapeType=Capsule, length=1.5358, radius=1.3486, center=<1.5717 0 0>, rotation=<rot 90 0 0>}]
 dyn_elbow_R [{shapeType=Capsule, length=2.1828, radius=1.1416, center=<2.5257 0 0>, rotation=<rot 90 0 0>}]
 dyn_shoulder_L [{shapeType=Capsule, length=0.5, radius=1, center=<1.0142 0 0>, rotation=<rot 90 0 0>}]
 dyn_shoulder_R [{shapeType=Capsule, length=1.3727, radius=0.8244, center=<1.2505 0 0>, rotation=<rot 90 0 0>}]
 dyn_wrist_L [{shapeType=Capsule, length=1.3, radius=1, center=<0.9174 0 0>, rotation=<rot 90 0 0>}]
 dyn_wrist_R [{shapeType=Capsule, length=1.1065, radius=1.0313, center=<1.6161 0 0>, rotation=<rot 90 0 0>}]
}
constraints [
 {bone1Name=dyn_Leg_L, bone2Name=base_COG, swing1LimitDegrees=12, swing2LimitDegrees=2, twistLimitDegrees=0}
 {bone1Name=dyn_Knee_L1, bone2Name=dyn_Leg_L, swing1LimitDegrees=15, swing2LimitDegrees=2, twistLimitDegrees=0}
 {bone1Name=dyn_ankle_L, bone2Name=dyn_Knee_L1, swing1LimitDegrees=6, swing2LimitDegrees=2, twistLimitDegrees=0}
 {bone1Name=dyn_Reg_R, bone2Name=base_COG, swing1LimitDegrees=12, swing2LimitDegrees=2, twistLimitDegrees=0}
 {bone1Name=dyn_Knee_R, bone2Name=dyn_Reg_R, swing1LimitDegrees=15, swing2LimitDegrees=2, twistLimitDegrees=0}
 {bone1Name=dyn_ankle_R, bone2Name=dyn_Knee_R, swing1LimitDegrees=8, swing2LimitDegrees=2, twistLimitDegrees=0}
 {bone1Name=dyn_shoulder_L, bone2Name=base_COG, swing1LimitDegrees=20, swing2LimitDegrees=5, twistLimitDegrees=0}
 {bone1Name=dyn_elbow_L, bone2Name=dyn_shoulder_L, swing1LimitDegrees=5, swing2LimitDegrees=10, twistLimitDegrees=0}
 {bone1Name=dyn_wrist_L, bone2Name=dyn_elbow_L, swing1LimitDegrees=5, swing2LimitDegrees=10, twistLimitDegrees=0}
 {bone1Name=dyn_shoulder_R, bone2Name=base_COG, swing1LimitDegrees=5, swing2LimitDegrees=20, twistLimitDegrees=0}
 {bone1Name=dyn_elbow_R, bone2Name=dyn_shoulder_R, swing1LimitDegrees=5, swing2LimitDegrees=10, twistLimitDegrees=0}
 {bone1Name=dyn_wrist_R, bone2Name=dyn_elbow_R, swing1LimitDegrees=5, swing2LimitDegrees=10, twistLimitDegrees=0}
]
